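# Physics asset for the DolphinGill backpack cosmetic, a sky-dive variant going by its name, that simulates its dangling slingshot and two bow chains.
# Capsule bodies cover each bone, with pack, pelvis and spine bodies locked to the root.
# The slingshot chain hangs from the pack with narrow swing on one axis and wider on the other.
# Left and right bow chains branch from a locked bag-handle anchor with limited swing and tight twist.
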PhysicsAsset /BRCosmetics/Accessories/FORT_Backpacks/Backpack_M_MED_DolphinGill/Meshes/M_MED_DolphinGill_Pack_Physics_sky_dive
bodies {
 based_bag_handle1 [{shapeType=Capsule, length=0, radius=0, shapeName=based_bag_handle1_capsule}]
 based_left_bow [{shapeType=Capsule, length=0.505, radius=0.505, shapeName=based_left_bow_capsule, rotation=<rot 0 -90 0>}]
 based_right_bow [{shapeType=Capsule, length=0.505, radius=0.505, shapeName=based_right_bow_capsule, rotation=<rot 0 -90 0>}]
 based_slingshot [{shapeType=Capsule, length=0.505, radius=0.505, shapeName=based_slingshot_capsule, rotation=<rot 0 -90 0>}]
 left_bow_0 [{
   shapeType=Capsule
   length=1.6308
   radius=2
   shapeName=left_bow_0_capsule
   center=<1.5975 -0.1431 0.3394>
   rotation=<rot 26.3796 92.0066 -65.4077>
  }]
 left_bow_1 [{shapeType=Capsule, length=1.725, radius=2.3946, shapeName=left_bow_1_capsule, center=<4.0156 0 0>, rotation=<rot -5.5309 -88.7032 -105.4295>}]
 left_bow_2 [{shapeType=Capsule, length=2.9474, radius=2.1034, shapeName=left_bow_2_capsule, center=<2.7831 0 0>, rotation=<rot 0 -85.943 -86.89>}]
 left_bow_3 [{shapeType=Capsule, length=2, radius=1.005, shapeName=left_bow_3_capsule, rotation=<rot 90 -90.0001 0>}]
 pack [
  {shapeType=Capsule, length=8.3024, radius=6.7842, shapeName=pack_capsule, center=<0.2561 -5.4805 13.6858>, rotation=<rot 11.6502 -90 88.4886>}
  {
   shapeType=Capsule
   length=14.1361
   radius=9.2959
   shapeName=pack_capsule
   center=<-0.0686 -6.5961 -7.4828>
   rotation=<rot 89.4671 -179.9611 0>
  }
  {
   shapeType=Capsule
   length=15.1122
   radius=7.9034
   shapeName=pack_capsule
   center=<-1.5206 -9.3037 -0.5513>
   rotation=<rot -89.2323 179.9933 -179.9935>
  }
 ]
 palvis_core [{shapeType=Capsule, length=16, radius=7, shapeName=palvis_core_capsule, center=<0 -1.397 -32.2484>, rotation=<rot -0.0001 -90 -90>}]
 right_bow_0 [{shapeType=Capsule, length=1.9276, radius=2.9721, shapeName=right_bow_0_capsule, center=<3.3478 0 0>, rotation=<rot -0.5701 15.9955 -89.7924>}]
 right_bow_1 [{
   shapeType=Capsule
   length=1.9722
   radius=0.9754
   shapeName=right_bow_1_capsule
   center=<2.3525 -0.0724 -0.0614>
   rotation=<rot 0 170.8587 -82.5065>
  }]
 right_bow_2 [{
   shapeType=Capsule
   length=2.7129
   radius=2.8765
   shapeName=right_bow_2_capsule
   center=<0.5228 0 0>
   rotation=<rot -5.3851 -176.7596 -104.9901>
  }]
 right_bow_3 [{shapeType=Capsule, length=4, radius=2.5629, shapeName=right_bow_3_capsule, center=<0.7594 0 0>, rotation=<rot -1.7045 4.6018 -89.9999>}]
 right_bow_4 [{shapeType=Capsule, length=1.505, radius=3.005, shapeName=right_bow_4_capsule, center=<1.3495 0 0>, rotation=<rot 0.0001 -90.0002 89.9999>}]
 root [{shapeType=Capsule, length=0.505, radius=0.505, shapeName=root_capsule, rotation=<rot 0 -90 0>}]
 slingshot_1 [{shapeType=Capsule, length=4.5, radius=1.5, shapeName=slingshot_1_capsule, center=<1.7303 0 0>, rotation=<rot 0.952 -87.4873 89.7758>}]
 slingshot_2 [{shapeType=Capsule, length=4.5, radius=1.5, shapeName=slingshot_2_capsule, center=<1.8819 0 0>, rotation=<rot 0.952 -87.4873 91.4271>}]
 slingshot_3 [{shapeType=Capsule, length=6.5, radius=1.5, shapeName=slingshot_3_capsule, center=<1.9037 0 0>, rotation=<rot 0.952 -87.4873 89.5439>}]
 slingshot_4 [{shapeType=Capsule, length=4.5, radius=1.5, shapeName=slingshot_4_capsule, center=<2.309 0 0>, rotation=<rot 0.952 -87.4873 89.6044>}]
 slingshot_5 [{shapeType=Capsule, length=4, radius=2, shapeName=slingshot_5_capsule, center=<2.2813 0 0>, rotation=<rot 0.952 -87.4873 89.7758>}]
 slingshot_6 [{shapeType=Capsule, length=0, radius=5.3222, shapeName=slingshot_6_capsule, center=<5.2287 0 0>}]
 slingshot_7 [{shapeType=Capsule, length=2, radius=1.5, shapeName=slingshot_7_capsule, rotation=<rot 0.2527 0.7946 90.9448>}]
 spine_core [{shapeType=Capsule, length=9, radius=5.505, shapeName=spine_core_capsule, center=<-5.1332 8.265 11.6852>, rotation=<rot -0.0008 -90 120>}]
}
constraints [
 {
  bone1Name=pack
  bone2Name=root
  swing1LimitDegrees=45
  swing1Motion=Locked
  swing2LimitDegrees=45
  swing2Motion=Locked
  twistLimitDegrees=45
  twistMotion=Locked
 }
 {
  bone1Name=based_slingshot
  bone2Name=pack
  swing1LimitDegrees=45
  swing1Motion=Locked
  swing2LimitDegrees=45
  twistLimitDegrees=45
  twistMotion=Locked
 }
 {bone1Name=slingshot_1, bone2Name=slingshot_7, swing1LimitDegrees=4, swing2LimitDegrees=32, twistLimitDegrees=10}
 {bone1Name=slingshot_2, bone2Name=slingshot_1, swing1LimitDegrees=4, swing2LimitDegrees=15, twistLimitDegrees=10}
 {bone1Name=slingshot_3, bone2Name=slingshot_2, swing1LimitDegrees=4, swing2LimitDegrees=15, twistLimitDegrees=10}
 {bone1Name=slingshot_4, bone2Name=slingshot_3, swing1LimitDegrees=4, swing2LimitDegrees=18, twistLimitDegrees=10}
 {bone1Name=slingshot_5, bone2Name=slingshot_4, swing1LimitDegrees=4, swing2LimitDegrees=25, twistLimitDegrees=10}
 {bone1Name=slingshot_6, bone2Name=slingshot_5, swing1LimitDegrees=4, swing2LimitDegrees=33, twistLimitDegrees=10}
 {bone1Name=left_bow_0, bone2Name=based_left_bow, swing1LimitDegrees=33, swing2LimitDegrees=33, twistLimitDegrees=11}
 {bone1Name=left_bow_1, bone2Name=left_bow_0, swing1LimitDegrees=33, swing2LimitDegrees=55, twistLimitDegrees=11}
 {bone1Name=right_bow_0, bone2Name=based_right_bow, swing1LimitDegrees=33, swing2LimitDegrees=33, twistLimitDegrees=11}
 {bone1Name=right_bow_1, bone2Name=right_bow_0, swing1LimitDegrees=33, swing2LimitDegrees=33, twistLimitDegrees=11}
 {bone1Name=right_bow_2, bone2Name=right_bow_1, swing1LimitDegrees=33, swing2LimitDegrees=33, twistLimitDegrees=11}
 {bone1Name=right_bow_3, bone2Name=right_bow_2, swing1LimitDegrees=33, swing2LimitDegrees=33, twistLimitDegrees=11}
 {bone1Name=left_bow_3, bone2Name=left_bow_2, swing1LimitDegrees=33, swing2LimitDegrees=33, twistLimitDegrees=6}
 {bone1Name=right_bow_4, bone2Name=right_bow_3, swing1LimitDegrees=33, swing2LimitDegrees=33, twistLimitDegrees=6}
 {
  bone1Name=palvis_core
  bone2Name=root
  swing1LimitDegrees=45
  swing1Motion=Locked
  swing2LimitDegrees=45
  swing2Motion=Locked
  twistLimitDegrees=45
  twistMotion=Locked
 }
 {
  bone1Name=spine_core
  bone2Name=root
  swing1LimitDegrees=45
  swing1Motion=Locked
  swing2LimitDegrees=45
  swing2Motion=Locked
  twistLimitDegrees=45
  twistMotion=Locked
 }
 {bone1Name=left_bow_2, bone2Name=left_bow_1, swing1LimitDegrees=33, swing2LimitDegrees=33, twistLimitDegrees=6}
 {bone1Name=slingshot_7, bone2Name=based_slingshot, swing1LimitDegrees=4, swing2LimitDegrees=22, twistLimitDegrees=10}
 {bone1Name=based_left_bow, bone2Name=based_bag_handle1, swing1LimitDegrees=22, swing2LimitDegrees=22, twistLimitDegrees=2}
 {bone1Name=based_right_bow, bone2Name=based_bag_handle1, swing1LimitDegrees=22, swing2LimitDegrees=22, twistLimitDegrees=2}
 {
  bone1Name=based_bag_handle1
  bone2Name=pack
  swing1LimitDegrees=45
  swing1Motion=Locked
  swing2LimitDegrees=45
  swing2Motion=Locked
  twistLimitDegrees=45
  twistMotion=Locked
 }
]
