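# Animation Blueprint for the DolphinGill backpack cosmetic, adding physics motion and keeping it pinned to the character.
# Input pose gets pack offsets, including a spine-driven one computed while skydiving, then chained AnimDynamics on clasps and handle, and is cached.
# Bool blends choose between RigidBody simulations of that cache for skydiving, front end and regular play.
# Finally, palvis_core and spine_core take world transforms from the attach parent's pelvis and spine_05.
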AnimBlueprint /BRCosmetics/Accessories/FORT_Backpacks/Backpack_M_MED_DolphinGill/Meshes/M_MED_DolphinGill_Pack_AnimBP
target_skeleton=&/M_MED_DolphinGill_Pack_Skeleton
graphs [
 AnimGraph {nodes [
   n1 = Local To Component(LocalPose=Use cached pose 'presim'())
   n2 = RigidBody!(
    ComponentPose="Transform (Modify) Bone - Bone: based_right_bow"(
     ComponentPose="Transform (Modify) Bone - Bone: pack"(ComponentPose=@n1, Translation="0,2.5,0", Rotation="0,0,0") {translationMode=BMM_Additive, boneToModify {boneName=pack}, translation {y=2.5}}
     Alpha=1
     Translation="0,0,0"
     Rotation="0,65,0"
     Scale="1,1,1"
    ) {rotationMode=BMM_Additive, translationMode=BMM_Additive, boneToModify {boneName=based_right_bow}, rotation {yaw=65}}
    Alpha
    OverrideWorldGravity:Vector="0,0,-980"
    ExternalForce
   ) -> (Pose) {
    bClampLinearTranslationLimitToRefPose=true
    bOverrideWorldGravity=true
    overridePhysicsAsset=&/M_MED_DolphinGill_Pack_Physics
    simulationSpace=BaseBoneSpace
    alphaBoolBlend {blendOption=Linear}
    baseBoneRef {boneName=pack}
    componentAppliedLinearAccClamp=<3000 3000 3000>
    simSpaceSettings {worldAlpha=1}
   }
   n3 = RigidBody(ComponentPose=Local To Component(LocalPose=Use cached pose 'presim'()), OverrideWorldGravity="0,0,-980") {
    bClampLinearTranslationLimitToRefPose=true
    bOverrideWorldGravity=true
    overridePhysicsAsset=&M_MED_DolphinGill_Pack_Physics
    simulationSpace=BaseBoneSpace
    alphaBoolBlend {blendOption=Linear}
    baseBoneRef {boneName=root}
    simSpaceSettings {
     dampingAlpha=0
     maxAngularAcceleration=12
     maxAngularVelocity=3
     maxLinearAcceleration=1550
     maxLinearVelocity=600
     worldAlpha=0.4832
     externalLinearDragV=<1 1 1>
    }
   }
   n4 = Blend Poses by bool(BlendPose_0=Component To Local(ComponentPose=@n2), BlendPose_1=Component To Local(ComponentPose=@n3), bActiveValue=Get bIsInFrontEnd())
   n5 = RigidBody(ComponentPose=Local To Component(LocalPose=Use cached pose 'presim'()), OverrideWorldGravity="0,0,-980") {
    bClampLinearTranslationLimitToRefPose=true
    bOverrideWorldGravity=true
    overridePhysicsAsset=&/M_MED_DolphinGill_Pack_Physics_sky_dive
    simulationSpace=BaseBoneSpace
    alphaBoolBlend {blendOption=Linear}
    baseBoneRef {boneName=root}
    simSpaceSettings {
     dampingAlpha=0
     maxAngularAcceleration=12
     maxAngularVelocity=3
     maxLinearAcceleration=1800
     maxLinearVelocity=600
     worldAlpha=0.4093
     externalLinearDragV=<1 1 1>
    }
   }
   n6 = Component To Local(ComponentPose=@n5)
   n7 = Get bIsSkyDiving()
   n8 = "Transform (Modify) Bone - Bone: palvis_core"!(
    ComponentPose=Local To Component(LocalPose=Blend Poses by bool(BlendPose_0=@n6, BlendPose_1=@n4, bActiveValue=@n7))
    bAlphaBoolEnabled:bool=True
    Translation=Get PelvisLoc() -> (PelvisLoc:Vector)
    Rotation=Get PelvisRot() -> (PelvisRot:Rotator)
    Scale=Get PelvisScale() -> (PelvisScale:Vector)
   ) -> (Pose) {
    alphaInputType=Bool
    rotationMode=BMM_Replace
    rotationSpace=BCS_WorldSpace
    scaleMode=BMM_Replace
    scaleSpace=BCS_WorldSpace
    translationMode=BMM_Replace
    translationSpace=BCS_WorldSpace
    alphaBoolBlend {blendInTime=0.2, blendOption=Linear, blendOutTime=0.2}
    boneToModify {boneName=palvis_core}
    rotation {roll=25}
   }
   n9 = "Transform (Modify) Bone - Bone: spine_core"!(
    ComponentPose=@n8
    bAlphaBoolEnabled:bool=True
    Translation=Get SpineLoc() -> (SpineLoc:Vector)
    Rotation=Get SpineRot() -> (SpineRot:Rotator)
    Scale=Get SpineScale() -> (SpineScale:Vector)
   ) -> (Pose) {
    alphaInputType=Bool
    rotationMode=BMM_Replace
    rotationSpace=BCS_WorldSpace
    scaleMode=BMM_Replace
    scaleSpace=BCS_WorldSpace
    translationMode=BMM_Replace
    translationSpace=BCS_WorldSpace
    alphaBoolBlend {blendInTime=0.2, blendOption=Linear, blendOutTime=0.2}
    boneToModify {boneName=spine_core}
   }
   Output Pose(Result=Component To Local(ComponentPose=@n9)) {layerGroup=None}
   n10 = Local To Component(LocalPose=Input Pose())
   n11 = "Transform (Modify) Bone - Bone: pack"(
    ComponentPose="Transform (Modify) Bone - Bone: based_right_bow"(
     ComponentPose="Transform (Modify) Bone - Bone: pack"(ComponentPose=@n10, Translation="0,0.5,0", Rotation="0,0,0") {translationMode=BMM_Additive, boneToModify {boneName=pack}, translation {y=0.5}}
     Alpha=1
     Translation="0,0,0"
     Rotation="0,65,0"
     Scale="1,1,1"
    ) {rotationMode=BMM_Additive, translationMode=BMM_Additive, boneToModify {boneName=based_right_bow}, rotation {yaw=65}}
    Translation=Get backpack_translation()
    Rotation=Get backpack_rotate()
   ) {rotationMode=BMM_Additive, translationMode=BMM_Additive, alphaBoolBlend {blendOption=Linear}, boneToModify {boneName=pack}}
   n12 = Anim Dynamics - Bone: clasp_0!(SimulationSpace:AnimPhysSimSpaceType Enum=World, ComponentPose=@n11, Alpha=1, ExternalForce="0,0,0") -> (Pose) {
    angularDampingOverride=0.9999
    angularSpringConstant=500
    bAngularSpring=true
    bOverrideAngularDamping=true
    bUseGravityOverride=true
    linearDampingOverride=0.9999
    simulationSpace=World
    alphaBoolBlend {blendOption=Linear}
    boundBone {boneName=clasp_0}
    componentAppliedLinearAccClamp=<3000 3000 3000>
    componentLinearAccScale {z=0.7}
    componentLinearVelScale {z=0.7}
    gravityOverride {z=-980}
    physicsBodyDefinitions [{
      boundBone {boneName=clasp_0}
      boxExtents=<10 10 10>
      constraintSetup {angularLimitsMax=<3 -3.9 40>, angularLimitsMin=<-3 -73.6 -43.6>, angularTarget=<1 0 0>}
      localJointOffset=<1 0 0>
     }]
    simSpaceSettings {maxAngularAcceleration=600, maxAngularVelocity=300, simSpaceAngularAlpha=1}
   }
   n13 = Anim Dynamics - Bone: clasp_1!(SimulationSpace:AnimPhysSimSpaceType Enum=World, ComponentPose=@n12, Alpha=1, ExternalForce="0,0,0") -> (Pose) {
    angularDampingOverride=0.9999
    angularSpringConstant=500
    bAngularSpring=true
    bOverrideAngularDamping=true
    bUseGravityOverride=true
    linearDampingOverride=0.9999
    simulationSpace=World
    alphaBoolBlend {blendOption=Linear}
    boundBone {boneName=clasp_1}
    componentAppliedLinearAccClamp=<3000 3000 3000>
    componentLinearAccScale {z=0.6}
    componentLinearVelScale {z=0.6}
    gravityOverride {z=-980}
    physicsBodyDefinitions [{
      boundBone {boneName=clasp_1}
      boxExtents=<10 10 10>
      constraintSetup {angularLimitsMax=<3 -3.9 40>, angularLimitsMin=<-3 -80.8 -32.8>, angularTarget=<1 0 0>}
      localJointOffset=<1 0 0>
     }]
    simSpaceSettings {maxAngularAcceleration=600, maxAngularVelocity=300, simSpaceAngularAlpha=1}
   }
   n14 = Anim Dynamics - Bone: clasp_2!(SimulationSpace:AnimPhysSimSpaceType Enum=World, ComponentPose=@n13, Alpha=1) -> (Pose) {
    angularDampingOverride=0.9999
    angularSpringConstant=500
    bAngularSpring=true
    bOverrideAngularDamping=true
    bUseGravityOverride=true
    linearDampingOverride=0.9999
    simulationSpace=World
    alphaBoolBlend {blendOption=Linear}
    boundBone {boneName=clasp_2}
    componentAppliedLinearAccClamp=<3000 3000 3000>
    componentLinearAccScale {z=0.5}
    componentLinearVelScale {z=0.5}
    gravityOverride {z=-980}
    physicsBodyDefinitions [{
      boundBone {boneName=clasp_2}
      boxExtents=<10 10 10>
      constraintSetup {angularLimitsMax=<1 1.9024 62.7805>, angularLimitsMin=<-1 -30 -4.2439>, angularTarget=<1 0 0>}
      localJointOffset=<1 0 0>
     }]
    simSpaceSettings {maxAngularAcceleration=600, maxAngularVelocity=300, simSpaceAngularAlpha=1}
   }
   n15 = Anim Dynamics - Bone: clasp_3!(SimulationSpace:AnimPhysSimSpaceType Enum=World, ComponentPose=@n14, Alpha=1, ExternalForce="0,0,0") -> (Pose) {
    angularDampingOverride=0.9999
    angularSpringConstant=500
    bAngularSpring=true
    bOverrideAngularDamping=true
    bUseGravityOverride=true
    linearDampingOverride=0.9999
    simulationSpace=World
    alphaBoolBlend {blendOption=Linear}
    boundBone {boneName=clasp_3}
    componentAppliedLinearAccClamp=<3000 3000 3000>
    componentLinearAccScale {z=0.7}
    componentLinearVelScale {z=0.7}
    gravityOverride {z=-980}
    physicsBodyDefinitions [{
      boundBone {boneName=clasp_3}
      boxExtents=<10 10 10>
      constraintSetup {angularLimitsMax=<1 30 18.878>, angularLimitsMin=<-1 -30 -3>, angularTarget=<1 0 0>}
      localJointOffset=<1 0 0>
     }]
    simSpaceSettings {maxAngularAcceleration=600, maxAngularVelocity=300, simSpaceAngularAlpha=1}
   }
   n16 = Anim Dynamics - Bone: clasp_4!(SimulationSpace:AnimPhysSimSpaceType Enum=World, ComponentPose=@n15, Alpha=1, ExternalForce="0,0,0") -> (Pose) {
    angularDampingOverride=0.9999
    angularSpringConstant=500
    bAngularSpring=true
    bOverrideAngularDamping=true
    bUseGravityOverride=true
    linearDampingOverride=0.9999
    simulationSpace=World
    alphaBoolBlend {blendOption=Linear}
    boundBone {boneName=clasp_4}
    componentAppliedLinearAccClamp=<3000 3000 3000>
    componentLinearAccScale {z=0.4}
    componentLinearVelScale {z=0.4}
    gravityOverride {z=-980}
    physicsBodyDefinitions [{
      boundBone {boneName=clasp_4}
      boxExtents=<10 10 10>
      constraintSetup {angularLimitsMax=<1 26.2245 4.9755>, angularLimitsMin=<-1 -13.7562 -6.5855>, angularTarget=<1 0 0>}
      localJointOffset=<1 0 0>
     }]
    simSpaceSettings {maxAngularAcceleration=600, maxAngularVelocity=300, simSpaceAngularAlpha=1}
   }
   n17 = Anim Dynamics - Bone: clasp_5!(SimulationSpace:AnimPhysSimSpaceType Enum=World, ComponentPose=@n16, Alpha=1, ExternalForce="0,0,0") -> (Pose) {
    angularDampingOverride=0.9999
    angularSpringConstant=500
    bAngularSpring=true
    bOverrideAngularDamping=true
    bUseGravityOverride=true
    linearDampingOverride=0.9999
    simulationSpace=World
    alphaBoolBlend {blendOption=Linear}
    boundBone {boneName=clasp_5}
    componentAppliedLinearAccClamp=<3000 3000 3000>
    componentLinearAccScale {z=0.7}
    componentLinearVelScale {z=0.7}
    gravityOverride {z=-980}
    physicsBodyDefinitions [{
      boundBone {boneName=clasp_5}
      boxExtents=<10 10 10>
      constraintSetup {angularLimitsMax=<1 2.0623 40>, angularLimitsMin=<-1 -68.0352 -40>, angularTarget=<1 0 0>}
      localJointOffset=<1 0 0>
     }]
    simSpaceSettings {maxAngularAcceleration=600, maxAngularVelocity=300, simSpaceAngularAlpha=1}
   }
   n18 = Anim Dynamics - Bone: bag_handle!(SimulationSpace:AnimPhysSimSpaceType Enum=Component, ComponentPose=@n17, Alpha=1, ExternalForce="0,0,0") -> (Pose) {
    angularDampingOverride=0.9999
    angularSpringConstant=500
    bAngularSpring=true
    bOverrideAngularDamping=true
    bUseGravityOverride=true
    linearDampingOverride=0.9999
    simulationSpace=World
    alphaBoolBlend {blendOption=Linear}
    boundBone {boneName=bag_handle}
    componentAppliedLinearAccClamp=<3000 3000 3000>
    componentLinearAccScale {z=0.7}
    componentLinearVelScale {z=0.7}
    physicsBodyDefinitions [{
      boundBone {boneName=bag_handle}
      boxExtents=<15 15 15>
      constraintSetup {angularLimitsMax=<1 10 2>, angularLimitsMin=<-1 -10 -2>, angularTarget=<1 0 0>}
      localJointOffset=<0.3 0 0>
     }]
    simSpaceSettings {maxAngularAcceleration=1800, maxAngularVelocity=600, simSpaceAngularAlpha=0.252}
   }
   Save cached pose 'presim'(Pose=Component To Local(ComponentPose=@n18)) {cacheName=presim}
  ]}
 EventGraph {nodes [
   Event BlueprintUpdateAnimation()
   > n1 = Sequence!(execute) -> (then_0, then_1, then_2:Exec, then_3:Exec, then_4:Exec)
   n2 = "Gravity Override State Function (Nativized)"!(
    execute=@n1.then_1
    CharacterPartAnimInstance:Custom Character Part Anim Instance Object Reference=Self-Reference()
    GravityOverrideParameterSet:Fort Gravity Override Parameter Set Structure=Get Gravity Override Params() -> (Gravity Override Params:Fort Gravity Override Parameter Set Structure)
    RigidBodyAnimNodeParameterSet:Fort Rigid Body Anim Node Parameter Set Structure
    DebugDraw:bool=false
    GameTimeInSeconds:float=GetGameTimeInSeconds()
    DebugJointName:Name=root
    StateLogicLayerName:Name=none
   ) -> (
    then
    OutGravityOverride:Vector
    OutCurrentState:Name=None
    OutGravityOverrideParameters:Fort Gravity Override Parameters Structure
    OutRigidBodyAnimNodeParameters:Fort Rigid Body Anim Node Parameters Structure
   )
   > Set GravityOverride(GravityOverride=Select(Option 0:Vector="0,0,-980", Option 1:Vector=@n2, Index=Get bIsSkyDiving()) -> (ReturnValue:Vector))
   n3 = GetPartSkeletalMeshComponent()
   Is Valid(exec=@n1.then_2, InputObject=@n3)
   n4 = GetAttachParent!(self:Scene Component Object Reference=@n3) -> (ReturnValue:Scene Component Object Reference)
   > Is Valid(InputObject=@n4)
   n5 = GetSocketTransform!(self=@n4, InSocketName=pelvis, TransformSpace=RTS_World) -> (ReturnValue_Location:Vector, ReturnValue_Rotation:Rotator, ReturnValue_Scale:Vector)
   > Set PelvisLoc(PelvisLoc:Vector=@n5) -> (Output_Get:Vector)
   > Set PelvisRot(PelvisRot:Rotator=@n5.ReturnValue_Rotation) -> (Output_Get:Rotator)
   > Set PelvisScale(PelvisScale:Vector=@n5.ReturnValue_Scale) -> (Output_Get:Vector)
   n6 = GetPartSkeletalMeshComponent()
   Is Valid(exec=@n1.then_4, InputObject=@n6)
   n7 = GetAttachParent(self=@n6)
   > Is Valid(InputObject=@n7)
   n8 = GetSocketTransform(self=@n7, InSocketName=spine_05, TransformSpace=RTS_World)
   > Set SpineLoc(SpineLoc:Vector=@n8) -> (Output_Get:Vector)
   > Set SpineScale(SpineScale:Vector=@n8.ReturnValue_Scale) -> (Output_Get:Vector)
   > Set SpineRot(SpineRot:Rotator=@n8.ReturnValue_Rotation) -> (Output_Get:Rotator)
   Branch(execute=@n1.then_0, Condition=Get bIsSkyDiving())
   n9 = GetOwnerSkeletalMeshComponent()
   n10 = BreakRotator(InRot=BreakTransform(InTransform=GetDeltaTransformFromRefPose(self=@n9, BoneName=spine_05, BaseName=spine_01)).Rotation)
   n11 = MapRangeClamped(Value=@n10.Yaw, InRangeA=-10, InRangeB=-50, OutRangeB=40)
   > Set backpack_rotate(backpack_rotate=MakeRotator(Roll=@n11, Pitch=float / float(A=@n11, B=5)))
   > Set backpack_translation(backpack_translation=MakeVector(Y=MapRangeClamped(Value=@n10.Yaw, InRangeA=-15, InRangeB=-50, OutRangeB=-10)))
   n12 = Branch(execute=@n1.then_3, Condition=Get bIsInFrontEnd())
   > Set SimlationSpace(SimlationSpace:AnimPhysSimSpaceType Enum=World) -> (Output_Get:AnimPhysSimSpaceType Enum)
   Set SimlationSpace(execute=@n12.else, SimlationSpace=Component)
   TryGetPawnOwner()
   TryGetPawnOwner()
  ]}
]
variables [
 alpha chains {type=number}
 backpack_rotate {type=Rotator}
 backpack_translation {type=Vector}
 Component Linear Acc Scale {type=Vector}
 Component Linear Vel Scale {type=Vector}
 Delta Time X {type=number}
 Gravity Override Params {
  type=FortGravityOverrideParameterSet
  default=(Skydive_GoingUp=(GravityOverride=(WindFrequency=(X=1.500000,Y=7.000000,Z=8.000000),WindAmplitude=(X=200.000000,Y=-500.000000,Z=400.000000),WindOffset=(X=0.000000,Y=-600.000000,Z=50.000000),GravityOverride=(X=0.000000,Y=0.000000,Z=-400.000000),JointName="root")),Skydive_FloatingDown=(GravityOverride=(WindFrequency=(X=12.000000,Y=12.000000,Z=12.000000),WindAmplitude=(X=50.000000,Y=-50.000000,Z=50.000000),WindOffset=(X=0.000000,Y=-50.000000,Z=200.000000),GravityOverride=(X=0.000000,Y=-600.000000,Z=-420.000000),JointName="root")),Skydive_DivingDown=(GravityOverride=(WindFrequency=(X=1.500000,Y=12.000000,Z=10.000000),WindAmplitude=(X=50.000000,Y=100.000000,Z=100.000000),GravityOverride=(X=-1500.000000,Y=0.000000,Z=0.000000),JointName="root")),Skydive_Parachuting=(GravityOverride=(WindFrequency=(X=1.500000,Y=7.000000,Z=8.000000),WindAmplitude=(X=100.000000,Y=-100.000000,Z=100.000000),GravityOverride=(X=0.000000,Y=-500.000000,Z=-980.000000),JointName="root")),OnGround_Standing=(GravityOverride=(GravityOverride=(X=0.000000,Y=0.000000,Z=-980.000000),JointName="root")),OnGround_Moving=(GravityOverride=(GravityOverride=(X=0.000000,Y=0.000000,Z=-980.000000),JointName="root")),OnGround_Moving_HighRotationRate=(GravityOverride=()),OnGround_Crouch_Moving=(GravityOverride=()),Emote_or_Melee=(GravityOverride=(GravityOverride=(X=0.000000,Y=0.000000,Z=-980.000000),JointName="root")),Swimming=(GravityOverride=()),Front_End=(GravityOverride=(GravityOverride=(X=0.000000,Y=0.000000,Z=-980.000000),JointName="root")),DBNO=(GravityOverride=(GravityOverride=(X=0.000000,Y=-450.000000,Z=-980.000000),JointName="root")),DBNO_Carried=(GravityOverride=(GravityOverride=(X=-980.000000,Y=500.000000,Z=-500.000000),JointName="pelvis")),OnGround_Falling=(GravityOverride=(GravityOverride=(X=0.000000,Y=0.000000,Z=-980.000000),JointName="root")),Ziplining=(GravityOverride=(GravityOverride=(X=0.000000,Y=0.000000,Z=-980.000000),JointName="root")),Skydive_Surfing=(GravityOverride=(GravityOverride=(X=0.000000,Y=0.000000,Z=-300.000000))),Riding_Standing=(GravityOverride=()),Riding_Moving=(GravityOverride=()),Riding_Moving_Targeting=(GravityOverride=()),Tactical_Sprint=(GravityOverride=(GravityOverride=(X=0.000000,Y=0.000000,Z=-980.000000),JointName="root")),Rail_Grinding=(GravityOverride=()),)
 }
 Gravity Override Params_0 {type=GravityOverrideParamsStruct}
 GravityOverride {type=Vector}
 NewVar {type=Vector}
 PelvisLoc {type=Vector}
 PelvisRot {type=Rotator}
 PelvisScale {type=Vector}
 SimlationSpace {type=AnimPhysSimSpaceType}
 SpineLoc {type=Vector}
 SpineRot {type=Rotator}
 SpineScale {type=Vector}
]
