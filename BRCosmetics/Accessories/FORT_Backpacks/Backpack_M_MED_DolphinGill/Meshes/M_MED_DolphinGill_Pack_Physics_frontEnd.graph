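# Front-end physics asset for the DolphinGill backpack cosmetic, defining secondary motion for its slingshot, two bows and bag handle.
# Capsule bodies cover each bone, with three forming the pack, which is rotation-locked to root.
# Swing and twist limits shape the slingshot and bow chains, while the handle swings mostly on one axis.
# The slingshot base has both a locked constraint to pack and a limited one to slingshot_scale; the bow bases have no parent constraint.
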PhysicsAsset /BRCosmetics/Accessories/FORT_Backpacks/Backpack_M_MED_DolphinGill/Meshes/M_MED_DolphinGill_Pack_Physics_frontEnd
bodies {
 bag_handle [{
   shapeType=Capsule
   length=3.1493
   radius=0.8266
   shapeName=bag_handle_capsule
   center=<-4.307 0.0615 -0.1748>
   rotation=<rot -1.7687 -90 -10.2492>
  }]
 based_bag_handle [{shapeType=Capsule, length=0.505, radius=0.505, shapeName=based_bag_handle_capsule, rotation=<rot 0 -90 0>}]
 based_left_bow [{shapeType=Capsule, length=0.505, radius=0.505, shapeName=based_left_bow_capsule, rotation=<rot 0 -90 0>}]
 based_right_bow [{shapeType=Capsule, length=0.505, radius=0.505, shapeName=based_right_bow_capsule, rotation=<rot 0 -90 0>}]
 based_slingshot [{shapeType=Capsule, length=0.505, radius=0.505, shapeName=based_slingshot_capsule, rotation=<rot 0 -90 0>}]
 left_bow_0 [{
   shapeType=Capsule
   length=1.2938
   radius=1.1308
   shapeName=left_bow_0_capsule
   center=<1.5975 -0.1431 0.3394>
   rotation=<rot 0 136.1582 -54.5484>
  }]
 left_bow_1 [{
   shapeType=Capsule
   length=1.225
   radius=1.8946
   shapeName=left_bow_1_capsule
   center=<0.7138 0.0914 0.0313>
   rotation=<rot 0 110.5473 -73.6313>
  }]
 pack [
  {shapeType=Capsule, length=8.3024, radius=6.7842, shapeName=pack_capsule, center=<1.4381 -5.1503 6.8536>, rotation=<rot 11.6503 -90 88.4886>}
  {
   shapeType=Capsule
   length=14.1361
   radius=9.2959
   shapeName=pack_capsule
   center=<0.1405 -5.5034 -15.1094>
   rotation=<rot 89.4671 -179.9611 -179.9611>
  }
  {
   shapeType=Capsule
   length=11.6122
   radius=7.9034
   shapeName=pack_capsule
   center=<-0.0435 -5.4171 -3.2456>
   rotation=<rot -89.2323 179.9933 -179.9935>
  }
 ]
 right_bow_0 [{
   shapeType=Capsule
   length=2.4276
   radius=2.4721
   shapeName=right_bow_0_capsule
   center=<3.3478 0.2224 0.0567>
   rotation=<rot 0 -54.0034 -89.3933>
  }]
 right_bow_1 [{
   shapeType=Capsule
   length=1.9722
   radius=0.9754
   shapeName=right_bow_1_capsule
   center=<2.3525 -0.0724 -0.0614>
   rotation=<rot 0 170.8587 -82.5065>
  }]
 right_bow_2 [{
   shapeType=Capsule
   length=1.7129
   radius=1.3765
   shapeName=right_bow_2_capsule
   center=<0.5228 0.0681 -0.0232>
   rotation=<rot 0 163.9505 -74.0741>
  }]
 right_bow_3 [{
   shapeType=Capsule
   length=1.2556
   radius=1.0629
   shapeName=right_bow_3_capsule
   center=<0.7594 -0.0672 -0.0984>
   rotation=<rot 0 -85.398 -88.2955>
  }]
 root [{shapeType=Capsule, length=0.505, radius=0.505, shapeName=root_capsule, rotation=<rot 0 -90 0>}]
 slingshot_0 [{
   shapeType=Capsule
   length=3.1525
   radius=2.5673
   shapeName=slingshot_0_capsule
   center=<-0.5751 -0.1115 0.8525>
   rotation=<rot -22.3428 3.0807 -89.3769>
  }]
 slingshot_1 [{shapeType=Capsule, length=1.9608, radius=1.7625, shapeName=slingshot_1_capsule, center=<3.5626 0 0>, rotation=<rot 0 -90 -90>}]
 slingshot_2 [{shapeType=Capsule, length=0.505, radius=0.505, shapeName=slingshot_2_capsule, rotation=<rot 0 -90 0>}]
 slingshot_3 [{shapeType=Capsule, length=2.5125, radius=1.8875, shapeName=slingshot_3_capsule, center=<0.6416 0 0>, rotation=<rot 0 -90 -90>}]
 slingshot_4 [{shapeType=Capsule, length=2.0394, radius=2.0919, shapeName=slingshot_4_capsule, center=<1.7923 0 0>, rotation=<rot 0 -90 -90>}]
 slingshot_5 [{shapeType=Capsule, length=3.2569, radius=1.8292, shapeName=slingshot_5_capsule, center=<3.0904 0 0>, rotation=<rot 0 -90 -90>}]
 slingshot_6 [{shapeType=Capsule, length=0.4865, radius=5.3222, shapeName=slingshot_6_capsule, center=<6.3328 0 0>, rotation=<rot 0 0 -90>}]
 slingshot_scale [{
   shapeType=Capsule
   length=6.1685
   radius=2.7272
   shapeName=slingshot_scale_capsule
   center=<-0.6834 0.0862 3.2494>
   rotation=<rot -11.0978 -90 176.7989>
  }]
}
constraints [
 {
  bone1Name=pack
  bone2Name=root
  swing1LimitDegrees=45
  swing1Motion=Locked
  swing2LimitDegrees=45
  swing2Motion=Locked
  twistLimitDegrees=45
  twistMotion=Locked
 }
 {
  bone1Name=based_slingshot
  bone2Name=pack
  swing1LimitDegrees=45
  swing1Motion=Locked
  swing2LimitDegrees=45
  swing2Motion=Locked
  twistLimitDegrees=45
  twistMotion=Locked
 }
 {bone1Name=slingshot_0, bone2Name=based_slingshot, swing1LimitDegrees=2, swing2LimitDegrees=2, twistLimitDegrees=2}
 {bone1Name=slingshot_1, bone2Name=slingshot_0, swing1LimitDegrees=33, swing2LimitDegrees=33, twistLimitDegrees=6}
 {bone1Name=slingshot_2, bone2Name=slingshot_1, swing1LimitDegrees=33, swing2LimitDegrees=33, twistLimitDegrees=6}
 {bone1Name=slingshot_3, bone2Name=slingshot_2, swing1LimitDegrees=33, swing2LimitDegrees=33, twistLimitDegrees=6}
 {bone1Name=slingshot_4, bone2Name=slingshot_3, swing1LimitDegrees=33, swing2LimitDegrees=33, twistLimitDegrees=6}
 {bone1Name=slingshot_5, bone2Name=slingshot_4, swing1LimitDegrees=33, swing2LimitDegrees=33, twistLimitDegrees=6}
 {bone1Name=slingshot_6, bone2Name=slingshot_5, swing1LimitDegrees=33, swing2LimitDegrees=14, twistLimitDegrees=6}
 {bone1Name=left_bow_0, bone2Name=based_left_bow, swing1LimitDegrees=33, swing2LimitDegrees=33, twistLimitDegrees=11}
 {bone1Name=left_bow_1, bone2Name=left_bow_0, swing1LimitDegrees=33, swing2LimitDegrees=55, twistLimitDegrees=11}
 {bone1Name=right_bow_0, bone2Name=based_right_bow, swing1LimitDegrees=33, swing2LimitDegrees=33, twistLimitDegrees=11}
 {bone1Name=right_bow_1, bone2Name=right_bow_0, swing1LimitDegrees=33, swing2LimitDegrees=33, twistLimitDegrees=11}
 {bone1Name=right_bow_2, bone2Name=right_bow_1, swing1LimitDegrees=33, swing2LimitDegrees=33, twistLimitDegrees=11}
 {bone1Name=right_bow_3, bone2Name=right_bow_2, swing1LimitDegrees=33, swing2LimitDegrees=33, twistLimitDegrees=11}
 {
  bone1Name=based_bag_handle
  bone2Name=pack
  swing1LimitDegrees=45
  swing1Motion=Locked
  swing2LimitDegrees=45
  swing2Motion=Locked
  twistLimitDegrees=45
  twistMotion=Locked
 }
 {bone1Name=bag_handle, bone2Name=based_bag_handle, swing1LimitDegrees=68.8, swing2LimitDegrees=7, twistLimitDegrees=2}
 {
  bone1Name=slingshot_scale
  bone2Name=pack
  swing1LimitDegrees=45
  swing1Motion=Locked
  swing2LimitDegrees=45
  swing2Motion=Locked
  twistLimitDegrees=45
  twistMotion=Locked
 }
 {bone1Name=based_slingshot, bone2Name=slingshot_scale, swing1LimitDegrees=22, swing2LimitDegrees=22, twistLimitDegrees=2}
]
